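# Unlit translucent Niagara sprite material for drop particles in the Rusty Raider Spark glider FX.
# Samples borrowed pickaxe lava noise at scaled UVs, offset per particle by a dynamic parameter whose U value drives both axes; V is unused.
# MF_ErosionNormalized erodes that noise using particle alpha times a radial falloff shaped by an exponent, so edges erode first.
# Particle RGB feeds the shared emissive adjustment function.
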Material /BRCosmetics/Gadgets/Assets/VinderTech_GliderChute/Glider_RustyRaider_Spark/FX/Materials/M_Glider_RustyRaider_Spark_Drops
graph {
 properties {bUsedWithNiagaraSprites=true, blendMode=BLEND_Translucent, pixelDepthOffsetMode=PDOM_Legacy, shadingModel=MSM_Unlit, usedShadingModels=Unlit}
 nodes [
  n1 = TextureCoordinate()
  n2 = AppendVector(A=ScalarParameter() {defaultValue=0.133, parameterName=UV Scale U}, B=ScalarParameter() {defaultValue=0.133, parameterName=UV Scale V})
  n3 = DynamicParameter() {paramNames [Random UV Offset U, Random UV Offset V, - unused -, - unused -]}
  n4 = TextureSample(UVs=Add(A=Multiply(A=@n1, B=@n2), B=AppendVector(A=@n3.Random UV Offset U, B=@n3.Random UV Offset U))) {
   samplerType=SAMPLERTYPE_LinearGrayscale
   texture=/BRCosmetics/Weapons/FORT_Melee/Pickaxe_RenegadeRaider_Spark/FX/Materials/T_Pickaxe_RenegadeRaider_Spark_LavaNoise
  }
  n5 = ParticleColor()
  n6 = ConstantBiasScale(None=@n1) {bias=-0.5, scale=2}
  n7 = Constant()
  n8 = OneMinus(None=Power(Base=Distance(A=@n6, B=@n7), Exp=ScalarParameter() {defaultValue=3, parameterName=Erosion Exponent}))
  MaterialFunctionCall!(Gradient=@n4.R, Erosion Amount=OneMinus(None=Multiply(A=@n5.A, B=@n8))) {materialFunction=/BRCosmetics/Athena/Effects/Particles/Skydiving/NS_Contrail_DarkAndLight/Functions/MF_ErosionNormalized}
  MaterialFunctionCall(Emissive=@n5.RGB) {materialFunction=&MF_FX_EmissiveAdjustments}
  OneMinus(None=@n5.A)
 ]
}
