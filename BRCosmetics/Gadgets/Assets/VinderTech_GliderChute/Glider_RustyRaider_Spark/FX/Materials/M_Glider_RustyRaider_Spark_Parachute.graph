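# Unlit modulate material for Niagara mesh particles on the Rusty Raider Spark glider parachute, tinting what lies behind with a hot HDR color.
# Panning lava noise on warped second-channel UVs is multiplied by the glider texture's red channel and particle alpha to blend from white to that color.
# A depth comparison fades it where the scene lies well behind the mesh.
# Vertices are offset along normals and toward the camera.
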Material /BRCosmetics/Gadgets/Assets/VinderTech_GliderChute/Glider_RustyRaider_Spark/FX/Materials/M_Glider_RustyRaider_Spark_Parachute
graph {
 properties {
  bUsedWithNiagaraMeshParticles=true
  blendMode=BLEND_Modulate
  pixelDepthOffsetMode=PDOM_Legacy
  shadingModel=MSM_Unlit
  translucencyPass=MTP_BeforeDOF
  usedShadingModels=Unlit
 }
 nodes [
  n1 = Subtract(A=CameraPositionWS(), B=WorldPosition().XYZ)
  Add(
   A=Multiply(A=VertexNormalWS(), B=ScalarParameter() {defaultValue=1, parameterName=Vertex Push})
   B=Multiply(A=Normalize(VectorInput=@n1), B=ScalarParameter() {defaultValue=2, parameterName=Camera Offset})
  )
  n2 = VectorParameter() {parameterName=Color, defaultValue=<rgba 9 1.95 0 2>}
  n3 = TextureSample() {texture=/BRCosmetics/Gadgets/Assets/VinderTech_GliderChute/Glider_RustyRaider_Spark/Textures/T_RustyRaider_Spark_D}
  n4 = MaterialFunctionCall(Float2=TextureCoordinate() {coordinateIndex=1}) {materialFunction=&BreakOutFloat2Components}
  n5 = OneMinus(None=@n4.G)
  n6 = ScalarParameter() {defaultValue=0.25, parameterName=Noise UV Exponent V}
  n7 = AppendVector(A=@n4.R, B=OneMinus(None=Power(Base=@n5, Exp=@n6)))
  n8 = AppendVector(A=ScalarParameter() {defaultValue=2, parameterName=Noise UV Scale U}, B=ScalarParameter() {defaultValue=1.5, parameterName=Noise UV Scale V})
  n9 = Constant()
  n10 = ScalarParameter() {defaultValue=0.333, parameterName=Noise UV Panner Speed}
  n11 = TextureSample(UVs=Panner(Coordinate=Multiply(A=@n7, B=@n8), Speed=AppendVector(A=@n9, B=@n10)) {bFractionalPart=true}) {
   samplerType=SAMPLERTYPE_LinearGrayscale
   texture=/BRCosmetics/Weapons/FORT_Melee/Pickaxe_RenegadeRaider_Spark/FX/Materials/T_Pickaxe_RenegadeRaider_Spark_LavaNoise
  }
  n12 = Power(Base=@n11.R, Exp=@n2.A)
  n13 = ParticleColor()
  n14 = PixelDepth()
  n15 = SceneDepth()
  n16 = Saturate(None=Add(A=Subtract(A=@n14.R, B=@n15.R), B=ScalarParameter() {defaultValue=22, parameterName=Pixel Depth Threshold}))
  LinearInterpolate(B=@n2.RGB, Alpha=Multiply(A=Multiply(A=@n3.R, B=@n12), B=Multiply(A=@n13.A, B=@n16))) {constA=1}
  Power(Base=Subtract(A=@n11.R, B=OneMinus(None=@n3.R)), Exp=@n2.A)
  ScalarParameter() {defaultValue=1, parameterName=Opacity}
  ScalarParameter() {defaultValue=3, parameterName=Mip Level}
 ]
}
