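# Unlit translucent material for Niagara ribbon and mesh trails on the Rusty Raider Spark glider, blending tip and base colors along U.
# Panned lava noise feeds MF_Erosion, shaped by width and length falloffs, driving color blend and opacity, with depth fade except on low and medium quality.
# Color alphas double as gradient length and exponent.
# Dynamic parameters control pan, offset and length erosion, though U offset feeds both axes, leaving V unused.
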Material /BRCosmetics/Gadgets/Assets/VinderTech_GliderChute/Glider_RustyRaider_Spark/FX/Materials/M_Glider_RustyRaider_Spark_Trail
graph {
 properties {
  bUsedWithNiagaraMeshParticles=true
  bUsedWithNiagaraRibbons=true
  blendMode=BLEND_Translucent
  pixelDepthOffsetMode=PDOM_Legacy
  shadingModel=MSM_Unlit
  translucencyPass=MTP_BeforeDOF
  usedShadingModels=Unlit
 }
 nodes [
  n1 = VectorParameter() {parameterName=Tip Color, defaultValue=<rgba 0.02 0.002 0.002 0.666>}
  n2 = VectorParameter() {parameterName=Base Color, defaultValue=<rgba 0.8 0.0933 0 1>}
  n3 = Saturate(None=TextureCoordinate())
  n4 = MaterialFunctionCall(Float2=@n3) {materialFunction=&BreakOutFloat2Components}
  n5 = Divide(A=@n4.R, B=@n1.A)
  n6 = Power(Base=OneMinus(None=Saturate(None=@n5)), Exp=@n2.A)
  n7 = MaterialFunctionCall(Float2=@n3) {materialFunction=&BreakOutFloat2Components}
  n8 = ScalarParameter() {defaultValue=0.666, parameterName=Noise UV Exponent U}
  n9 = Multiply(
   A=AppendVector(A=Power(Base=@n7.R, Exp=@n8), B=@n7.G)
   B=AppendVector(
    A=ScalarParameter() {defaultValue=0.666, parameterName=Noise UV Scale U}
    B=ScalarParameter() {defaultValue=0.333, parameterName=Noise UV Scale V}
   )
  )
  n10 = DynamicParameter() {defaultValue=<rgba 0 0 1 1>, paramNames [Noise UV Panner V Speed, Noise UV Offset U, Noise UV Offset V, Erosion Length Exponent]}
  n11 = AppendVector(A=@n10.Noise UV Offset U, B=@n10.Noise UV Offset U)
  n12 = ScalarParameter() {defaultValue=-1.1, parameterName=Noise UV Panner Speed}
  n13 = TextureSample(UVs=Panner(Coordinate=Add(A=@n9, B=@n11), Speed=AppendVector(A=@n12, B=@n10.Noise UV Panner V Speed)) {bFractionalPart=true}) {
   samplerType=SAMPLERTYPE_LinearGrayscale
   texture=/BRCosmetics/Weapons/FORT_Melee/Pickaxe_RenegadeRaider_Spark/FX/Materials/T_Pickaxe_RenegadeRaider_Spark_LavaNoise
  }
  n14 = ConstantBiasScale(None=@n4.G) {bias=-0.5, scale=2}
  n15 = OneMinus(None=Power(Base=Abs(None=@n14), Exp=ScalarParameter() {defaultValue=0.45, parameterName=Erosion Width Exponent}))
  n16 = OneMinus(None=Power(Base=@n4.R, Exp=@n10.Erosion Length Exponent))
  n17 = MaterialFunctionCall!(
   Gradient=@n13.R
   Erosion Amount=OneMinus(None=Multiply(A=@n15, B=@n16))
   Softness=ScalarParameter() {defaultValue=0.25, parameterName=Erosion Softness}
  ) {materialFunction=/BRCosmetics/Athena/Effects/Particles/Skydiving/NS_Contrail_DarkAndLight/Functions/MF_Erosion}
  MaterialFunctionCall(Emissive=LinearInterpolate(A=@n1.RGB, B=LinearInterpolate(A=@n1.RGB, B=@n2.RGB, Alpha=@n6), Alpha=@n17.Default Range)) {materialFunction=&MF_FX_EmissiveAdjustments}
  QualitySwitch(
   Default=DepthFade(Opacity=@n17.Default Range, FadeDistance=ScalarParameter() {defaultValue=2, parameterName=Depth Fade Distance})
   Low=@n17.Default Range
   Medium=@n17.Default Range
  )
 ]
}
