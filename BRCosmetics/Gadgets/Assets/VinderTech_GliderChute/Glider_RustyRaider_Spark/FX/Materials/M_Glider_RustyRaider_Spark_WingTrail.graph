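# Additive unlit Niagara ribbon and mesh-particle material for the Rusty Raider Spark glider's wing trail.
# Panned lava noise, UV-warped along V and offset by a dynamic parameter (U drives both axes), is masked toward the U edges.
# MF_Erosion erodes it along the length, particle alpha fades it, and the result drives a black-to-hot-orange four-color emissive ramp.
# Separate branches compute a camera-directed offset and an upward-normal opacity term.
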Material /BRCosmetics/Gadgets/Assets/VinderTech_GliderChute/Glider_RustyRaider_Spark/FX/Materials/M_Glider_RustyRaider_Spark_WingTrail
graph {
 properties {
  bUsedWithNiagaraMeshParticles=true
  bUsedWithNiagaraRibbons=true
  blendMode=BLEND_Additive
  pixelDepthOffsetMode=PDOM_Legacy
  shadingModel=MSM_Unlit
  translucencyPass=MTP_BeforeDOF
  usedShadingModels=Unlit
 }
 nodes [
  n1 = TextureCoordinate() {coordinateIndex=1}
  n2 = MaterialFunctionCall(Float2=@n1) {materialFunction=&BreakOutFloat2Components}
  n3 = ScalarParameter() {defaultValue=0.666, parameterName=Noise UV Exponent}
  n4 = Multiply(
   A=AppendVector(A=@n2.R, B=Power(Base=@n2.G, Exp=@n3))
   B=AppendVector(A=ScalarParameter() {defaultValue=0.5, parameterName=Noise UV Scale U}, B=ScalarParameter() {defaultValue=1, parameterName=Noise UV Scale V})
  )
  n5 = AppendVector(A=Constant(), B=ScalarParameter() {defaultValue=-0.8, parameterName=Noise UV Panner Speed})
  n6 = DynamicParameter() {paramNames [UV Offset U, UV Offset V, "", ""]}
  n7 = TextureSample(UVs=Add(A=Panner(Coordinate=@n4, Speed=@n5) {bFractionalPart=true}, B=AppendVector(A=@n6.UV Offset U, B=@n6.UV Offset U))) {
   samplerType=SAMPLERTYPE_LinearGrayscale
   texture=/BRCosmetics/Weapons/FORT_Melee/Pickaxe_RenegadeRaider_Spark/FX/Materials/T_Pickaxe_RenegadeRaider_Spark_LavaNoise
  }
  n8 = ConstantBiasScale(None=ComponentMask(None=@n1) {r=true}) {bias=-0.5, scale=2}
  n9 = Multiply(A=Power(Base=@n7.R, Exp=ScalarParameter() {defaultValue=3, parameterName=Noise Exponent}), B=OneMinus(None=DotProduct(A=@n8, B=@n8)))
  n10 = ComponentMask(None=@n1) {g=true}
  n11 = OneMinus(None=Power(Base=OneMinus(None=@n10), Exp=ScalarParameter() {defaultValue=2, parameterName=Erosion Length Exponent}))
  n12 = ScalarParameter() {defaultValue=1, parameterName=Erosion Softness}
  n13 = Saturate(None=ParticleColor().A)
  n14 = Saturate(None=Subtract(
    A=MaterialFunctionCall!(Gradient=@n9, Erosion Amount=@n11, Softness=@n12) {materialFunction=/BRCosmetics/Athena/Effects/Particles/Skydiving/NS_Contrail_DarkAndLight/Functions/MF_Erosion}.Increased Range
    B=OneMinus(None=@n13)
   ))
  MaterialFunctionCall(Emissive=MaterialFunctionCall(
    1=Constant()
    2=VectorParameter() {parameterName=End Color, defaultValue=<rgba 0.02 0 0 0.666>}.RGB
    3=VectorParameter() {parameterName=Middle Color, defaultValue=<rgba 0.333 0.0363 0.0033 1>}.RGB
    4=VectorParameter() {parameterName=Base Color, defaultValue=<rgba 3.33 0.4179 0.0333 1>}.RGB
    A=SquareRoot(None=@n14)
   ) {materialFunction=&Lerp_Multiple_Float3}.Lerp 4 Inputs) {materialFunction=&MF_FX_EmissiveAdjustments}
  n15 = CameraPositionWS()
  n16 = WorldPosition()
  Multiply(A=Normalize(VectorInput=Subtract(A=@n15, B=@n16.XYZ)), B=ScalarParameter() {defaultValue=1, parameterName=Camera Offset})
  n17 = ComponentMask(None=VertexNormalWS()) {b=true}
  Multiply(A=SquareRoot(None=Saturate(None=@n17)), B=ScalarParameter() {defaultValue=0.666, parameterName=Opacity})
 ]
}
